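# Niagara system for the WingBath Sturdy emote that draws rocket smoke and fire ribbon trails from a skeletal mesh.
# In both emitters, five spawn groups each select a bone or socket and a ribbon ID, giving separate trails.
# Minimal001 makes widening gray smoke that rises with curl noise and erodes, with lifetime scaled by platform quality.
# RocketFire001 adds thin, short-lived ribbons cooling from hot orange to gray, with motion forces disabled.
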NiagaraSystem /BRCosmetics/Animation/Game/MainPlayer/Emotes/WingBath_Sturdy/VFX/NS_Emote_WingBath_Sturdy_SmokeTrail
dependencies {
 usedDataInterfaces [&NiagaraDataInterfaceCurve, &NiagaraDataInterfacePlatformSet, &NiagaraDataInterfaceColorCurve]
 usedDynamicInputs []
 usedModules [
  &self:SystemSpawnScript.SystemScriptSource.SystemScriptGraph.NiagaraNodeAssignment_0.SetVariables_#1
  &self:SystemSpawnScript.SystemScriptSource.SystemScriptGraph.NiagaraNodeAssignment_1.SetVariables_#2
  &SystemState
 ]
}
emitters {
 Minimal001 {
  emitterUpdateScript [
   &EmitterState(Scalability Mode=::NewEnumerator1)
   &SpawnBurst_Instantaneous(
    Spawn Count=&Multiply_Int(A=$System.RocketNumber, B=$System.ParticlePerRibbons)
    Use Spawn Probability=false
    UseLoopCountLimit=false
   ) {enabled=false}
   &SpawnRate(
    SpawnRate=&FloatFromCurve(
     FloatCurve=DataInterface() {propertyValues {Curve {
        defaultValue=1
        keys [
         {arriveTangent=-0.0405, interpMode=RCIM_Cubic, leaveTangent=-0.0405, tangentMode=RCTM_User, time=0, value=1.5}
         {arriveTangent=-4.5342, interpMode=RCIM_Cubic, leaveTangent=-4.5342, tangentMode=RCTM_User, time=0.2, value=1}
        ]
       }}}
     CurveIndex=$Emitter.NormalizedLoopAge
     Scale Curve=15
    )
    Use Spawn Probability=false
   )
   &SpawnRate(
    SpawnGroup=1
    SpawnRate=&FloatFromCurve(
     FloatCurve=DataInterface() {propertyValues {Curve {
        defaultValue=1
        keys [
         {arriveTangent=-0.0405, interpMode=RCIM_Cubic, leaveTangent=-0.0405, tangentMode=RCTM_User, time=0, value=1.5}
         {arriveTangent=-4.5342, interpMode=RCIM_Cubic, leaveTangent=-4.5342, tangentMode=RCTM_User, time=0.2, value=1}
        ]
       }}}
     CurveIndex=$Emitter.NormalizedLoopAge
     Scale Curve=15
    )
    Use Spawn Probability=false
   ) {module=SpawnRate001}
   &SpawnRate(
    SpawnGroup=2
    SpawnRate=&FloatFromCurve(
     FloatCurve=DataInterface() {propertyValues {Curve {
        defaultValue=1
        keys [
         {arriveTangent=-0.0405, interpMode=RCIM_Cubic, leaveTangent=-0.0405, tangentMode=RCTM_User, time=0, value=1.5}
         {arriveTangent=-4.5342, interpMode=RCIM_Cubic, leaveTangent=-4.5342, tangentMode=RCTM_User, time=0.2, value=1}
        ]
       }}}
     CurveIndex=$Emitter.NormalizedLoopAge
     Scale Curve=15
    )
    Use Spawn Probability=false
   ) {module=SpawnRate002}
   &SpawnRate(
    SpawnGroup=3
    SpawnRate=&FloatFromCurve(
     FloatCurve=DataInterface() {propertyValues {Curve {
        defaultValue=1
        keys [
         {arriveTangent=-0.0405, interpMode=RCIM_Cubic, leaveTangent=-0.0405, tangentMode=RCTM_User, time=0, value=1.5}
         {arriveTangent=-4.5342, interpMode=RCIM_Cubic, leaveTangent=-4.5342, tangentMode=RCTM_User, time=0.2, value=1}
        ]
       }}}
     CurveIndex=$Emitter.NormalizedLoopAge
     Scale Curve=15
    )
    Use Spawn Probability=false
   ) {module=SpawnRate004}
   &SpawnRate(
    SpawnGroup=4
    SpawnRate=&FloatFromCurve(
     FloatCurve=DataInterface() {propertyValues {Curve {
        defaultValue=1
        keys [
         {arriveTangent=-0.0405, interpMode=RCIM_Cubic, leaveTangent=-0.0405, tangentMode=RCTM_User, time=0, value=1.5}
         {arriveTangent=-4.5342, interpMode=RCIM_Cubic, leaveTangent=-4.5342, tangentMode=RCTM_User, time=0.2, value=1}
        ]
       }}}
     CurveIndex=$Emitter.NormalizedLoopAge
     Scale Curve=15
    )
    Use Spawn Probability=false
   ) {module=SpawnRate003}
  ]
  particleSpawnScript [
   &PartitionParticles(
    Number Of Partitions=$System.RocketNumber
    Particles Per Partition:NiagaraInt32=$System.ParticlePerRibbons
    Partition Mode:ENiagara_PartitionMode_v2=::NewEnumerator1
    Write Partition Index to Ribbon Link Order:ENiagara_AttributeSamplingApplyOutput=::NewEnumerator0
    Write Partition to Ribbon ID=::NewEnumerator0
   ) {enabled=false}
   /SetVariables(Particles.RibbonID=/Niagara/DynamicInputs/Meshes/MakeRibbonIDFromSpawnGroup()) {isSetParameters=true, module=SetVariables_#3}
   &SkeletalMeshLocation(
    Bone / Socket Index=&SpawnGroup()
    MaskSpawnGroup=false
    Mesh Sampling Type=::NewEnumerator5
    Sampled Orientation Coordinate Space=::Local
    Sampled Position Coordinate Space=::Local
    Sampled Position Offset Coordinate Space:ENiagaraSkelMeshTransforms=::NewEnumerator2
    Sampled Velocity Coordinate Space:ENiagaraCoordinateSpace=::Local
    Skeletal Mesh=$System.Skeletal Mesh
    Socket Sampling Mode=::NewEnumerator5
   ) {module=SkeletalMeshLocation002}
   &InitializeParticle(
    Lifetime=&FN_PickFloatDependOnPlatform(
     TargetPlatform=DataInterface() {propertyValues {Platforms {qualityLevelMask=-8}}}
     TargetPlatformActive=2
     TargetPlatformNoActive=&FN_PickFloatDependOnPlatform(
      TargetPlatform=DataInterface() {propertyValues {Platforms {qualityLevelMask=-28}}}
      TargetPlatformActive=1.5
      TargetPlatformNoActive=&FN_PickFloatDependOnPlatform(
       TargetPlatform=DataInterface() {propertyValues {Platforms {qualityLevelMask=-30}}}
       TargetPlatformActive=1
       TargetPlatformNoActive=&FN_PickFloatDependOnPlatform(TargetPlatform=DataInterface() {propertyValues {Platforms {qualityLevelMask=-31}}}, TargetPlatformActive=0.75, TargetPlatformNoActive=0)
      )
     )
    )
    Ribbon Width=&RandomRangeFloat(Minimum=10, Maximum=20)
    Ribbon Width Mode=::NewEnumerator1
    Sprite Size Mode=::NewEnumerator3
    UsePositionOffset=false
    Write Material Random=true
   )
   &AddVelocity() {enabled=false}
   /SetVariables(Particles.Position=&AddVectorToPosition(Position=$Particles.SkeletalMeshLocation002.SampledPosition, Vector=<0 0 0>)) {isSetParameters=true, module=SetVariables_#4}
   /SetVariables(Particles.UVDistance:NiagaraFloat=&Multiply_Float(A=&VectorLength(Vector=$Particles.SkeletalMeshLocation002.SampledVelocity), B=0.5)) {isSetParameters=true}
  ]
  particleUpdateScript [
   &ParticleState()
   /SetVariables(Particles.RibbonUVDistance:NiagaraFloat=$Particles.UVDistance) {isSetParameters=true}
   &ScaleRibbonWidth(Ribbon Width Scale=&FloatFromCurve(
     FloatCurve=DataInterface() {propertyValues {Curve {keys [
         {arriveTangent=2.331, interpMode=RCIM_Cubic, leaveTangent=2.331, tangentMode=RCTM_User, time=0, value=0.2}
         {arriveTangent=0.2911, interpMode=RCIM_Cubic, leaveTangent=0.2911, tangentMode=RCTM_User, time=0.0438, value=0.275}
         {arriveTangent=1.32, interpMode=RCIM_Cubic, leaveTangent=1.32, tangentMode=RCTM_User, time=0.5234, value=1.9578}
         {arriveTangent=0.3935, interpMode=RCIM_Cubic, leaveTangent=0.3935, tangentMode=RCTM_User, time=1, value=2.5}
        ]}}}
     Scale Curve=&RandomRangeFloat(Minimum=1, Maximum=1.5)
    )) {module=ScaleRibbonWidth001}
   &CurlNoiseForce(
    Bias Noise=false
    Noise Frequency=5
    Noise Strength=&FloatFromCurve(
     FloatCurve=DataInterface() {propertyValues {Curve {
        defaultValue=1
        keys [<at 0 0 RCIM_Cubic>, {arriveTangent=1.5, interpMode=RCIM_Cubic, leaveTangent=1.5, time=0.7, value=0.2}, <at 1 1.5 RCIM_Cubic>]
       }}}
     Scale Curve=75
    )
    Pan Noise=false
   )
   &FN_RibbonMaterial(
    Color=&ColorFromCurve(ColorCurve=DataInterface() {propertyValues {
       AlphaCurve {keys [<at 0 2 RCIM_Cubic>,<at 0.0513 2>]}
       BlueCurve {keys [<at 0.0258 0.5089>]}
       GreenCurve {keys [<at 0.0258 0.4793>]}
       RedCurve {keys [<at 0.0258 0.4735>]}
      }})
    Erosion Amount:NiagaraFloat=&Lerp_Float(A=0.25, B=1, Alpha=$Particles.NormalizedAge)
    Erossion Scale U=4.5
    Macro Distortion Amount=0.1
    Macro Pan Offset=-0.1216
    Micro Distortion Amount=0.0384
    Opacity Erosion:NiagaraFloat=&FloatFromCurve(FloatCurve=DataInterface() {propertyValues {Curve {keys [<at 0 0.25 RCIM_Cubic>, {arriveTangent=1.1077, tangentMode=RCTM_User, time=1, value=1}]}}})
    Opacity Power:NiagaraFloat=&FloatFromCurve(FloatCurve=DataInterface() {propertyValues {Curve {keys [<at 0 0.55 RCIM_Cubic>, {arriveTangent=1.1077, tangentMode=RCTM_User, time=1, value=1}]}}})
    Scale Alpha:NiagaraFloat=&FloatFromCurve(
     FloatCurve=DataInterface() {propertyValues {Curve {keys [
         <at 0 0.7 RCIM_Cubic>
         {arriveTangent=-0.6429, interpMode=RCIM_Cubic, leaveTangent=-0.6429, time=0.2, value=0.45}
         {arriveTangent=-0.5592, interpMode=RCIM_Cubic, leaveTangent=-0.5592, tangentMode=RCTM_User, time=0.7, value=0.25}
         {arriveTangent=-0.4965, tangentMode=RCTM_User, time=1, value=0}
        ]}}}
     Scale Curve=&FloatFromCurve(FloatCurve=DataInterface() {propertyValues {Curve {
         defaultValue=1
         keys [
          <at 0 0 RCIM_Cubic>
          {arriveTangentWeight=0.0471, interpMode=RCIM_Cubic, leaveTangentWeight=0.4243, tangentMode=RCTM_User, time=0.1, value=1}
          <at 1 1 RCIM_Cubic>
         ]
        }}})
    )
    UV Basis:NiagaraFloat=&ReturnNormalizedExecIndex()
   ) {module=FN_RibbonMaterial001}
   &GravityForce(Gravity=&Multiply_VectorByFloat(
     Vector=<0 0 1>
     Float=&FloatFromCurve(
      FloatCurve=DataInterface() {propertyValues {Curve {
         defaultValue=1
         keys [<at 0 0 RCIM_Cubic>, {arriveTangent=1.5, interpMode=RCIM_Cubic, leaveTangent=1.5, time=0.5, value=0.6}, <at 1 1.5 RCIM_Cubic>]
        }}}
      Scale Curve=50
     )
    ))
   &SolveForcesAndVelocity(Clamp Velocity=false, Limit Acceleration=false)
  ]
  rendererClasses [&NiagaraRibbonRendererProperties]
  renderers [{rendererClass=&NiagaraRibbonRendererProperties}]
 }
 RocketFire001 {
  emitterUpdateScript [
   &EmitterState()
   &SpawnBurst_Instantaneous(
    Spawn Count=&Multiply_Int(A=$System.RocketNumber, B=$System.ParticlePerRibbons)
    Use Spawn Probability=false
    UseLoopCountLimit=false
   ) {enabled=false}
   &SpawnRate(SpawnRate=80, Use Spawn Probability=false)
   &SpawnRate(SpawnGroup=1, SpawnRate=80, Use Spawn Probability=false) {module=SpawnRate001}
   &SpawnRate(SpawnGroup=2, SpawnRate=80, Use Spawn Probability=false) {module=SpawnRate002}
   &SpawnRate(SpawnGroup=3, SpawnRate=80, Use Spawn Probability=false) {module=SpawnRate004}
   &SpawnRate(SpawnGroup=4, SpawnRate=80, Use Spawn Probability=false) {module=SpawnRate003}
  ]
  particleSpawnScript [
   /SetVariables(Particles.RibbonID=&MakeRibbonIDFromSpawnGroup()) {isSetParameters=true, module=SetVariables_#3}
   &SkeletalMeshLocation(
    Bone / Socket Index=&SpawnGroup()
    MaskSpawnGroup=false
    Mesh Sampling Type=::NewEnumerator5
    Sampled Orientation Coordinate Space=::Local
    Sampled Position Coordinate Space=::Local
    Sampled Position Offset Coordinate Space:ENiagaraSkelMeshTransforms=::NewEnumerator2
    Sampled Velocity Coordinate Space:ENiagaraCoordinateSpace=::Local
    Skeletal Mesh=$System.Skeletal Mesh
    Socket Sampling Mode=::NewEnumerator5
   ) {module=SkeletalMeshLocation002}
   &InitializeParticle(Lifetime=0.2, Ribbon Width=4, Ribbon Width Mode=::NewEnumerator1, UsePositionOffset=false, Write Material Random=true)
   /SetVariables(Particles.Position=&AddVectorToPosition(Position=$Particles.SkeletalMeshLocation002.SampledPosition, Vector=<0 0 0>)) {isSetParameters=true, module=SetVariables_#4}
  ]
  particleUpdateScript [
   &ParticleState()
   &ScaleRibbonWidth(Ribbon Width Scale=&FloatFromCurve(FloatCurve=DataInterface() {propertyValues {Curve {keys [
         {arriveTangent=3.7788, interpMode=RCIM_Cubic, leaveTangent=3.7788, tangentMode=RCTM_User, time=0, value=0}
         {arriveTangent=-0.8732, interpMode=RCIM_Cubic, leaveTangent=-0.8732, tangentMode=RCTM_User, time=0.1019, value=0.4}
         {arriveTangent=-0.2676, interpMode=RCIM_Cubic, leaveTangent=-0.2676, tangentMode=RCTM_User, time=1, value=0}
        ]}}})) {module=ScaleRibbonWidth001}
   &FN_RibbonMaterial(
    Color=&ColorFromCurve(ColorCurve=DataInterface() {propertyValues {
       AlphaCurve {keys [<at 0 2>]}
       BlueCurve {keys [<at 0 1 RCIM_Cubic>,<at 0.5151 0.5115>]}
       GreenCurve {keys [<at 0 5 RCIM_Cubic>,<at 0.5151 0.4809>]}
       RedCurve {keys [<at 0 25 RCIM_Cubic>,<at 0.5151 0.4751>]}
      }})
    Color Pan U Speed=0.8342
    Erosion Amount:NiagaraFloat=&Lerp_Float(A=0, B=1, Alpha=$Particles.NormalizedAge)
    Erossion Scale U=4.5
    Macro Distortion Amount=0.1
    Macro Pan Speed=0.064
    Micro Distortion Amount=0.0384
    Micro Pan Speed=-0.1344
    Opacity Erosion:NiagaraFloat=&FloatFromCurve(FloatCurve=DataInterface() {propertyValues {Curve {keys [<at 0 0 RCIM_Cubic>, {arriveTangent=1.1077, tangentMode=RCTM_User, time=1, value=1}]}}})
    Opacity Pan U Speed=0.1408
    Opacity Power:NiagaraFloat=&FloatFromCurve(FloatCurve=DataInterface() {propertyValues {Curve {keys [<at 0 0.55 RCIM_Cubic>, {arriveTangent=1.1077, tangentMode=RCTM_User, time=1, value=1}]}}})
    Scale Alpha:NiagaraFloat=&FloatFromCurve(
     FloatCurve=DataInterface() {propertyValues {Curve {keys [
         <at 0 0.5 RCIM_Cubic>
         {arriveTangent=-0.9109, interpMode=RCIM_Cubic, leaveTangent=-0.9109, tangentMode=RCTM_User, time=0.4, value=0.225}
         {arriveTangent=-0.369, tangentMode=RCTM_User, time=1, value=0}
        ]}}}
     Scale Curve=0.5
    )
   ) {module=FN_RibbonMaterial001}
   &CurlNoiseForce(
    Bias Noise=false
    Noise Frequency=20
    Noise Strength=&FloatFromCurve(
     FloatCurve=DataInterface() {propertyValues {Curve {
        defaultValue=1
        keys [<at 0 0 RCIM_Cubic>, {arriveTangent=1.5, interpMode=RCIM_Cubic, leaveTangent=1.5, time=0.4, value=0.8}, <at 1 1.5 RCIM_Cubic>]
       }}}
     Scale Curve=50
    )
    Pan Noise=false
   ) {enabled=false}
   &SolveForcesAndVelocity(Clamp Velocity=false, Limit Acceleration=false) {enabled=false}
  ]
  rendererClasses [&NiagaraRibbonRendererProperties]
  renderers [{rendererClass=&NiagaraRibbonRendererProperties}]
 }
}
summary {
 systemName=NS_Emote_WingBath_Sturdy_SmokeTrail
 emitters [
  {emitterName=Minimal001, rendererClasses [&NiagaraRibbonRendererProperties]}
  {emitterName=RocketFire001, rendererClasses [&NiagaraRibbonRendererProperties]}
 ]
 userVariables [
  User.bisFootHidden {type=NiagaraBool}
  User.bIsFrontEnd {type=NiagaraBool}
  User.bIsFrontEndPreview {type=NiagaraBool}
  User.Debug {type=NiagaraBool}
  User.DebugVelocity {type=Vector3f, description="", defaultValue {struct=Vector3f, value=<0 0 1>}}
  User.EmissiveColor {type=LinearColor, description="", defaultValue {struct=LinearColor, value=<rgba 0.0196 0.2745 0.5255 1>}}
  User.ParaGlideLeanAlpha {type=NiagaraFloat, defaultValue=0, description=""}
  User.SkeletalMesh {
   type=NiagaraDataInterfaceSkeletalMesh
   defaultValue {
    struct=NiagaraExt_VariableValue_DataInterface
    value {dataInterface=&self:NiagaraDataInterfaceSkeletalMesh_0}
   }
  }
 ]
}
