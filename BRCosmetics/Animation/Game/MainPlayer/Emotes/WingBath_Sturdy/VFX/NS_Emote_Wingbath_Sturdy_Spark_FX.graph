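# Niagara spark effect for the Wingbath Sturdy emote, with Flare, SmokeLWE and two Spark_Large emitters enabled, whose modules and renderers are not shown.
# The only detailed emitter, NE_FN_Sparks, is disabled: a GPU sprite burst thrown out under drag, gravity and fading noise.
# It derives color and mass from a random SparkIntensity, scales size by LOD distance and speed, and cools color over life.
# User parameters expose front-end flags and light intensity.
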NiagaraSystem /BRCosmetics/Animation/Game/MainPlayer/Emotes/WingBath_Sturdy/VFX/NS_Emote_Wingbath_Sturdy_Spark_FX
dependencies {
 usedDataInterfaces [&NiagaraDataInterfaceCurve, &NiagaraDataInterfaceColorCurve]
 usedDynamicInputs [&VectorFromFloat, &MakeCustomLinearColorFromBool]
 usedModules [
  &self:SystemSpawnScript.SystemScriptSource.SystemScriptGraph.NiagaraNodeAssignment_1.SetVariables_#1
  &SystemState
 ]
}
emitters {
 Flare {rendererClasses [], renderers []}
 NE_FN_Sparks {
  bEnabled=false
  simTarget=GPUComputeSim
  emitterUpdateScript [
   &EmitterState()
   &SpawnBurst_Instantaneous(Spawn Count=50, Use Spawn Probability=false)
  ]
  particleSpawnScript [
   /SetVariables(Particles.SparkIntensity:NiagaraFloat=&FloatFromCurve(
     FloatCurve=DataInterface() {propertyValues {Curve {keys [
         <at 0 1 RCIM_Cubic>
         {arriveTangent=-2.5317, interpMode=RCIM_Cubic, leaveTangent=-2.5317, tangentMode=RCTM_User, time=0.6, value=0.5}
         <at 1 0 RCIM_Cubic>
        ]}}}
     CurveIndex=&RandomRangeFloat(Minimum=0, Maximum=1)
    )) {isSetParameters=true}
   &InitializeParticle(
    Color=&LerpLinearColors(StartColor=<rgba 10 1.5225 0.5208 1>, EndColor=<rgba 25 16.8856 2.9948 1>, LerpFactor=$Particles.SparkIntensity)
    Lifetime Max=0.25
    Lifetime Min=0.1
    Lifetime Mode=::NewEnumerator1
    Mass=&Lerp_Float(A=0.3, B=1, Alpha=$Particles.SparkIntensity)
    Mass Mode=::NewEnumerator1
    Sprite Size Max=<2.7 2>
    Sprite Size Min=<1.6 1.125>
    Sprite Size Mode=::NewEnumerator2
    UsePositionOffset=false
    Write Material Random=true
   )
   /SetVariables(Particles.SpriteSize=&Multiply_Vector2(
     A=$Particles.SpriteSize
     B=&Vector2DFromFloat(Value=&FloatFromCurve(FloatCurve=DataInterface() {propertyValues {Curve {keys [<at 1000 1 RCIM_Cubic>,<at 5000 2>]}}}, CurveIndex=$Engine.Owner.LODDistance))
    )) {isSetParameters=true}
   &SphereLocation(
    MaskSpawnGroup=false
    Override Seed=false
    Sphere Radius=3
    Surface Only=false
    UseNonUniformScale=false
    UseOffset=false
    UseSphereOrientation=false
   )
   &AddVelocityInCone(
    Cone Angle=180
    Cone Axis=<0 0 1>
    Cone Axis Coordinate Space=::World
    Module.Use Velocity Falloff On Cone Axis=false
    Velocity Strength=&RandomRangeFloat(Minimum=70, Maximum=450)
   )
   &ApplyInitialForces()
  ]
  particleUpdateScript [
   &ParticleState()
   &Drag(Drag=&RandomRangeFloat(Minimum=0.5, Maximum=2.3), Use Linear Drag=true, Use Rotational Drag=false)
   &GravityForce(Gravity=<0 0 -700>)
   &VectorNoiseForce(
    Force Amount=&FloatFromCurve(
     FloatCurve=DataInterface() {propertyValues {Curve {keys [
         {arriveTangent=-9.4515, interpMode=RCIM_Cubic, leaveTangent=-4.1687, tangentMode=RCTM_User, time=0, value=1}
         {arriveTangent=-3.4372, interpMode=RCIM_Cubic, leaveTangent=-3.4372, tangentMode=RCTM_User, time=0.6, value=0}
        ]}}}
     Scale Curve=3000
    )
    Module.Use Random Noise Time Interval=false
   )
   &SolveForcesAndVelocity(Clamp Velocity=false, Limit Acceleration=false)
   &ScaleColor(
    Linear Color Curve=DataInterface() {propertyValues {
      AlphaCurve {keys [<at 0 1>,<at 1 1>]}
      BlueCurve {keys [
        <at 0 2 RCIM_Cubic>
        {arriveTangent=-1.452, interpMode=RCIM_Cubic, leaveTangent=-1.452, tangentMode=RCTM_User, time=0.1467, value=1}
        <at 1 0.0755>
       ]}
      GreenCurve {keys [
        <at 0 2 RCIM_Cubic>
        {arriveTangent=-1.4408, interpMode=RCIM_Cubic, leaveTangent=-1.4408, tangentMode=RCTM_User, time=0.1467, value=1}
        <at 1 0.078>
       ]}
      RedCurve {keys [
        <at 0 2 RCIM_Cubic>
        {arriveTangent=-0.8332, interpMode=RCIM_Cubic, leaveTangent=-0.8332, tangentMode=RCTM_User, time=0.1467, value=1}
        <at 1 0.25>
       ]}
     }}
    Scale Mode=::NewEnumerator2
   )
   &ScaleSpriteSize(Scale Factor=&Vector2DFromFloat(Value=&FloatFromCurve(FloatCurve=DataInterface() {propertyValues {Curve {keys [<at 0.4 1 RCIM_Cubic>, {arriveTangent=-3.8662, interpMode=RCIM_Cubic, leaveTangent=-3.8662, tangentMode=RCTM_User, time=1, value=0}]}}})))
   &ScaleSpriteSizeBySpeed(
    Max Scale Factor=<0.8 3.4>
    Sample Scale Factor By Curve=true
    Scale Factor Curve=DataInterface() {propertyValues {Curve {keys [
        {arriveTangent=0.2757, time=0, value=0}
        {arriveTangent=0.3844, interpMode=RCIM_Cubic, leaveTangent=0.3844, tangentMode=RCTM_User, time=0.3, value=0.15}
        {arriveTangent=2.2765, interpMode=RCIM_Cubic, leaveTangent=2.2765, tangentMode=RCTM_User, time=1, value=1}
       ]}}}
    Velocity Threshold=350
   )
  ]
  rendererClasses [&NiagaraSpriteRendererProperties]
  renderers [{rendererClass=&NiagaraSpriteRendererProperties}]
 }
 SmokeLWE {rendererClasses [], renderers []}
 Spark_Large_04 {rendererClasses [], renderers []}
 Spark_Large_04001 {rendererClasses [], renderers []}
 flareBackup {bEnabled=false, rendererClasses [], renderers []}
}
summary {
 systemName=NS_Emote_Wingbath_Sturdy_Spark_FX
 emitters [
  {emitterName=Flare, rendererClasses []}
  {emitterName=Spark_Large_04001, rendererClasses []}
  {emitterName=Spark_Large_04, rendererClasses []}
  {emitterName=SmokeLWE, rendererClasses []}
  {bEnabled=false, emitterName=flareBackup, rendererClasses []}
  {bEnabled=false, emitterName=NE_FN_Sparks, simTarget=GPUComputeSim, rendererClasses [&NiagaraSpriteRendererProperties]}
 ]
 userVariables [
  User.bIsFrontEnd {type=NiagaraBool}
  User.bIsFrontEndPreview {type=NiagaraBool}
  User.LightIntensity {type=NiagaraFloat, defaultValue=0, description=""}
 ]
}
